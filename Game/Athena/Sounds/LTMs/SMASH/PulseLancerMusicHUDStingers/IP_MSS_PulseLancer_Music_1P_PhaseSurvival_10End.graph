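# One-shot stereo MetaSound music stinger for the PulseLancer LTM, named for the end of a survival-phase countdown.
# On play, it starts the countdown-end wave and an ADSR envelope that multiplies both channels, giving a slow fade-in to full level.
# The wave's completion ends the source.
# It uses the base music sound class and sends to sidechain volume buses for footsteps, ambience, creatures, explosions, vehicles and weapons.
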MetaSoundSource /Game/Athena/Sounds/LTMs/SMASH/PulseLancerMusicHUDStingers/IP_MSS_PulseLancer_Music_1P_PhaseSurvival_10End
inputs [UE.Source.OnPlay {type=Trigger}]
interfaces [UE.OutputFormat.Stereo:1.0, UE.Source.OneShot:1.0, UE.Source:1.1]
outputs [
 "UE.OutputFormat.Stereo.Audio:0" {type=Audio}
 "UE.OutputFormat.Stereo.Audio:1" {type=Audio}
 UE.Source.OneShot.OnFinished {type=Trigger}
]
pages [Default {
  input_defaults {UE.Source.OnPlay=null}
  nodes [
   ADSREnvelopeADSREnvelope_0 = ADSR Envelope.ADSR Envelope:Float(Trigger Attack=^UE.Source.OnPlay) {also {defaults {Attack Time=2, Sustain Level=1}}}
   WavePlayer_0 = UE.Wave Player:Stereo(Play=^UE.Source.OnPlay) {also {defaults {Wave Asset=&/IP_MX_S37_PulseLancer_RoundCountdownEnd_STNG}, version [1,1]}}
   Multiply_0 = UE.Multiply:Audio by Float(AdditionalOperands=@ADSREnvelopeADSREnvelope_0.Out Envelope, PrimaryOperand=@WavePlayer_0.Out Right)
   Multiply_1 = UE.Multiply:Audio by Float(AdditionalOperands=@ADSREnvelopeADSREnvelope_0.Out Envelope, PrimaryOperand=@WavePlayer_0.Out Left)
  ]
  links [
   {dst="@UE.OutputFormat.Stereo.Audio:1", src=Multiply_0.Out}
   {dst="@UE.OutputFormat.Stereo.Audio:0", src=Multiply_1.Out}
   {dst="@UE.Source.OneShot.OnFinished", src=WavePlayer_0.On Finished}
  ]
 }]
properties {
 assetImportData=&self:AssetImportData
 numChannels=2
 outputFormat=Stereo
 sampleRate=48000
 soundClassObject=/Game/Sounds/FortSoundClassesAndModes/Fort_11_Base_Music.Fort_11_Base_music
 volume=1.2
 assetUserData [&self:AudioAssetUserData_1]
 busSends [
  {audioBus=/Game/Sounds/Modulation/EnvelopeFollowers/Volume/AB_Sidechain_Footsteps_Vol, sendLevel=1}
  {audioBus=/Game/Sounds/Modulation/EnvelopeFollowers/Volume/AB_SideChain_FortAmbient_Outdoors_Vol, sendLevel=1}
  {audioBus=&AB_Sidechain_FortAmbient_Vol, sendLevel=1}
  {audioBus=&AB_Sidechain_FortCreatures_Vol, sendLevel=1}
  {audioBus=&AB_Sidechain_FortExplosion_Vol, sendLevel=1}
  {audioBus=&AB_Sidechain_FortVehicles_Vol, sendLevel=1}
  {audioBus=/Game/Sounds/Modulation/EnvelopeFollowers/Volume/AB_Sidechain_FortWeapon1P_Vol, sendLevel=1}
  {audioBus=&AB_Sidechain_FortWeapon_Vol, sendLevel=1}
 ]
}
